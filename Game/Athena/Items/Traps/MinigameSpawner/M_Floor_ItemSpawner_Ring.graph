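# Additive, unlit, two-sided mesh-particle material for the glowing ring of the minigame floor item spawner, with responsive AA.
# A SphereMask on the U axis of the second UV channel forms a band centered at the midpoint, sized by the particle-driven Radius dynamic parameter and softened by Hardness.
# The mask blends Color Low into the brighter HDR Color High, and MF_FX_EmissiveAdjustments processes the result as emissive.
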Material /Game/Athena/Items/Traps/MinigameSpawner/M_Floor_ItemSpawner_Ring
graph {
 properties {
  bEnableResponsiveAA=true
  bUsedWithMeshParticles=true
  blendMode=BLEND_Additive
  pixelDepthOffsetMode=PDOM_Legacy
  previewMesh=/Game/Effects/Fort_Effects/Mesh/Splines/SM_Spline_Ring
  shadingModel=MSM_Unlit
  twoSided=true
  usedShadingModels=Unlit
 }
 nodes [
  n1 = ComponentMask(None=TextureCoordinate() {coordinateIndex=1}) {r=true}
  n2 = Constant() {r=0.5}
  n3 = DynamicParameter() {defaultValue=<rgba 0.15 1 1 1>, paramNames [Radius,Param2,Param3,Param4]}
  n4 = ScalarParameter() {parameterName=Hardness}
  MaterialFunctionCall(Emissive=LinearInterpolate(
    A=VectorParameter() {parameterName=Color Low, defaultValue=<rgba 0.0097 0.0093 0.1615 1>}.RGB
    B=VectorParameter() {parameterName=Color High, defaultValue=<rgba 0.2497 0.9677 6 1>}.RGB
    Alpha=SphereMask(A=@n1, B=@n2, Radius=@n3.Radius, Hardness=@n4)
   )) {materialFunction=&MF_FX_EmissiveAdjustments}
 ]
}
