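; Actor component for Creative item spawners that spawns or grants items from a replicated spawn list built from spawnable actors.
; Authority-checked SpawnNext picks entries sequentially or randomly, gated by simultaneous limits, blocking and pauses, on a self-restarting non-looping timer.
; Spawn async-loads items, then configures pickup team, class and resource cost.
; Bound pickup delegates track live pickups, restart the timer, grant spare ammo and fire dispatchers.
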(event EventEndPlay
  (params (EndPlayReason :type EEndPlayReason%20Enum :default "Destroyed"))
  (CallFunction|ResetSpawner true))

(event Custom|Spawn
  (params (SpawnParams :type Minigame%20Spawner%20Spawn%20Params%20Structure%20%28by%20ref%29))
  (bind (_pickuptospawn _pickupquantity _pickupinstigatorhandle _itemvariantguid _spawntransform) (Utilities|Struct|BreakMinigameSpawnerSpawnParams SpawnParams))
  (Variables|Default|SetbBlockSpawning true)
  (if (Utilities|IsValidSoftObjectReference _pickuptospawn)
    (Math|Integer|IncrementInt (Variables|Default|GetNumItemsSpawned)
      (:%20%20
        (bind _object (Utilities|AsyncLoadAsset _pickuptospawn)
          (:Completed
            (bind _asfort_world_item_definition (Utilities|Casting|CastToFortWorldItemDefinition _object)
              (:then
                (Utilities|IsItemDefintionValid _asfort_world_item_definition (Variables|Default|GetInvalidItemTags)
                  (:IsValid
                    (bind _returnvalue (CallFunction|SpawnPickupfromSpawnParams SpawnParams (Variables|Default|GetBasePickupClass) (.location _spawntransform) (Math|Rotator|GetRotationXVector (.rotation _spawntransform)) (Variables|Default|GetbShouldTossPickups) true false (Variables|Default|GetSourceType) "ItemSpawner"))
                    (Utilities|IsValid _returnvalue
                      (:Is%20Valid
                        (Utilities|FlowControl|Sequence
                          (:then_0
                            ;[1 1 1] Bind Pickup delegates if the spawned pickup class doesn't already have them
                            (Utilities|Casting|CastToFortPickupCreative _returnvalue
                              (:then
                                (if (Utilities|DoesObjectImplementInterface (Components|GetOwner) "/Script/FortniteGame.FortPickupEventHandler")
                                  _
                                  (else
                                    _)))
                              (:CastFailed
                                _))
                            (Pickup|SetupPickup _returnvalue))
                          (:then_1
                            (Default|CallOnSpawnDispatcher _returnvalue)
                            (Transformation|SetActorScale3D _returnvalue (Variables|Default|GetItemSpawnScale))
                            ;[1 1 1] This could destory the item, so this should be after the bindings
                            (Utilities|IsValid _returnvalue
                              (:Is%20Valid
                                (bind _asfort_pickup_creative (Utilities|Casting|CastToFort_Pickup_Creative _returnvalue))
                                (Variables|/Script/FortniteGame.FortPickupCreative|SetbPickupOnTouch (Variables|Default|GetIsPickupOnTouch) _asfort_pickup_creative))))
                          (:then_2
                            ;[1 1 1] Add cost to pickup item if applicable
                            (if (Math|Boolean|ORBoolean (> (Variables|Default|GetItemWoodCost) 0) (> (Variables|Default|GetItemMetalCost) 0) (> (Variables|Default|GetItemStoneCost) 0) (> (Variables|Default|GetItemGoldCost) 0))
                              (Utilities|IsValid _returnvalue
                                (:Is%20Valid
                                  (bind _asfort_pickup_creative_2 (Utilities|Casting|CastToFort_Pickup_Creative _returnvalue))
                                  (Utilities|IsValid _asfort_pickup_creative_2
                                    (:Is%20Valid
                                      (Class|FortPickupCreative|SetItemResourceCost _asfort_pickup_creative_2 "Stone" (Variables|Default|GetItemStoneCost))
                                      (Class|FortPickupCreative|SetItemResourceCost _asfort_pickup_creative_2 "Metal" (Variables|Default|GetItemMetalCost))
                                      (Class|FortPickupCreative|SetItemResourceCost _asfort_pickup_creative_2 "Wood" (Variables|Default|GetItemWoodCost))
                                      (Class|FortPickupCreative|SetItemResourceCost _asfort_pickup_creative_2 "GoldCurrency" (Variables|Default|GetItemGoldCost))))))))
                          (:then_3
                            _)))
                      (:Is%20Not%20Valid
                        ;[1 1 1] Reverse our incremenation of num items spawned upon failure
                        (Math|Integer|DecrementInt (Variables|Default|GetNumItemsSpawned)
                          (:%20%20
                            _))))
                    ;[1 1 1] Unblock Spawning
                    (Variables|Default|SetbBlockSpawning (Variables|Default|GetbUseManualSpawnControl))
                    (if (Variables|Default|GetbIsFinishedSpawning)
                      (Default|CallOnLastItemFinishedSpawning)))
                  (:IsNotValid
                    _)))
              (:CastFailed
                _))))))
    (else
      _)))

(event Events|EventNotifyItemContainerChanged
  (Default|CallOnItemContainerUpdated))

(fn AddToSpawnList
  (params (ActorToSpawn :type Actor%20Object%20Reference) (SpawnLocation :type Vector))
  (bind _actorspawntransform (LocalVariables|GetActorSpawnTransform))
  (bind _asfort_spawnable_item_interface (Utilities|Casting|CastToFortSpawnableItemInterface ActorToSpawn)
    (:then
      (bind _output_get (LocalVariables|SetActor%20Item%20Interface _asfort_spawnable_item_interface))
      (LocalVariables|SetCached%20Spawn%20List (Variables|Minigame|GetToSpawnList))
      (if (SpawnableItemInterface|CanbeAddedtoSpawnList _output_get)
        (LocalVariables|SetActorSpawnTransform (CallFunction|MakeSpawnTransform (Variables|GetActorToSpawn) (Variables|GetSpawnLocation)))
        ;[0 1 0.9331 0.25] Add All of Spawn Entries associated with the interface to the Cached Spawn List
        (for _array_element (SpawnableItemInterface|GetSpawnItemEntries (LocalVariables|GetActor%20Item%20Interface) (Variables|Default|GetInvalidItemTags))
          (if (.ReturnValue (CallFunction|MakeSpawnParamsfromItemEntry _array_element _actorspawntransform))
            (Utilities|Array|Add (LocalVariables|GetCached%20Spawn%20List) (CallFunction|MakeSpawnParamsfromItemEntry _array_element _actorspawntransform))))
        ;[0 0.3696 1 0.25] Replicates the spawn list and calls delegates.
        (bind _returnvalue (CallFunction|SettoSpawnListwithLimit (LocalVariables|GetCached%20Spawn%20List))) ; Setting the spawn list once all the items have been added\nThis avoids unnecessary replication with each addition.
        (if _returnvalue
          (SpawnableItemInterface|OnAddedtoSpawnList (LocalVariables|GetActor%20Item%20Interface))
          (Default|CallAddedtoSpawnListDispatcher (Variables|GetActorToSpawn))
          (Networking|FlushNetDormancy (Components|GetOwner)) ; Ensure Spawn Array replicates Not guaranteed as components go dormant easily..
          (return true)
          (else
            ;;[1 0.5233 0.4534] The NewToSpawnList should only get too big when a creator maliciously tries really hard to cause a performance issue by mass dropping items on to an item spawner for minutes. Since the failure is not likely to happen and not really harmful, it is fine to let it go without logging
            (return false)))
        (else
          _)))
    (:CastFailed
      _))
  (return false))

(fn AttemptGrant
  (params (OptionalInventoryOwner :type Fort%20Inventory%20Owner%20Interface%20Interface%20%28by%20ref%29) (MinigameSpawnerSpawnParams :type Minigame%20Spawner%20Spawn%20Params%20Structure%20%28by%20ref%29))
  (bind _minigamespawnerspawnparams (Variables|GetMinigameSpawnerSpawnParams))
  (bind (_pickuptospawn _pickupquantity _pickupinstigatorhandle _itemvariantguid _weaponammooverride) (Utilities|Struct|BreakMinigameSpawnerSpawnParams _minigamespawnerspawnparams))
  (bind _output (Utilities|ResolveSoftReference _pickuptospawn))
  (if (Utilities|IsValidInterface OptionalInventoryOwner)
    (bind (_outerrormessage _returnvalue) (Class|FortKismetLibrary|IsTargetUnobstructedWithinRangeforPickup (.location (Utilities|Struct|BreakMinigameSpawnerSpawnParams MinigameSpawnerSpawnParams)) (Variables|GetOptionalInventoryOwner)))
    (if _returnvalue
      (if (> (Inventory|GetInventorySpace (Variables|GetOptionalInventoryOwner)) 0)
        (bind _returnvalue_2 (Inventory|GiveItemtoInventoryOwner (Variables|GetOptionalInventoryOwner) _output _itemvariantguid _pickupquantity true -1 _pickupinstigatorhandle true _weaponammooverride))
        (Utilities|IsValid _returnvalue_2
          (:Is%20Valid
            (CallFunction|GiveAdditionalAmmo (Variables|GetOptionalInventoryOwner) _output (Variables|Default|GetSpareWeaponAmmo) (Variables|Default|GetSpareWeaponAmmo_Override))
            (Utilities|FlowControl|Sequence
              (:then_0
                (bind _asfort_world_item_definition (Utilities|Casting|CastToFortWorldItemDefinition (WeaponItemDefinition|GetLinkedItemDefinition _output)))
                (CallFunction|GiveAdditionalAmmo 0 _asfort_world_item_definition (Variables|Default|GetSpareLinkedWeaponAmmo) (Variables|Default|GetSpareLinkedWeaponAmmo_Override)))
              (:then_1
                (Default|CallOnGrantDispatcher)
                (return true))))
          (:Is%20Not%20Valid
            (return false)))
        (else
          _))
      (else
        (Development|PrintString (Utilities|String|Append "Minigame Spawner: Skipped Granting because " _outerrormessage))))
    (return false)
    (else
      _)))

(fn BindPickupDelegates
  (params (Pickup :type Fort%20Pickup%20Object%20Reference))
  (bind _pickup (Variables|GetPickup)
    (:then
      (Interaction|BindEventtoOnPickup _pickup (EventDispatchers|CreateEvent :!event OnPickup))
      (Game|BindEventtoOnDestroyed _pickup (EventDispatchers|CreateEvent :!event OnPickupDestroyed))
      (Interaction|BindEventtoOnPickupCombinedOverflow _pickup (EventDispatchers|CreateEvent :!event OnPickupCombinedOverflow))
      (Interaction|BindEventtoOnPickupCombined _pickup (EventDispatchers|CreateEvent :!event OnPickupCombined)))
    (:else)))

(fn CanSpawn
  (return (Math|Boolean|ANDBoolean (CallFunction|IsNextSpawnIndexValid) (CallFunction|HasRoomtoSpawn) (or (Variables|Default|GetbIgnoreSpawnTimer) (CallFunction|IsSpawning)) (not (Variables|Default|GetbBlockSpawning)) (not (Variables|Default|GetbBlockedByPawn)))))

(fn CopySpawnListArray
  (params (SpawnListToCopy :type Array%20of%20Minigame%20Spawner%20Spawn%20Params%20Structures))
  (for _array_element SpawnListToCopy
    (Utilities|Array|Add (Variables|Default|GetRemainingSpawnList) _array_element))
  (Variables|Default|SetHasToSpawnArrayBeenCopied true)
  (return (Variables|Default|GetRemainingSpawnList)))

(fn DestroyPickups
  (bind _spawnedpickups (Variables|Default|GetSpawnedPickups))
  (for _array_element _spawnedpickups
    (Utilities|IsValid _array_element
      (:Is%20Valid
        (Actor|DestroyActor _array_element))))
  (Utilities|Array|Clear _spawnedpickups))

(fn FinishSpawning
  (Variables|Default|SetbIsFinishedSpawning true)
  (Utilities|Time|ClearandInvalidateTimerbyHandle (Variables|Default|GetSpawnTimer)))

(fn ForceSpawnRemaining
  (params (bRestartTimer :type Boolean :default true))
  (Variables|Default|SetbHasBegunSpawning true)
  (if bRestartTimer
    (CallFunction|RestartTimer)
    (else
      _))
  (CallFunction|SpawnNext) ; Do one immediate spawn; important to do this after setting timer so that if there is only one spawn SpawnNext can invalidate the timer
)

(fn GetLatestSpawnedPickup
  (return (Utilities|Array|Get(acopy) (Variables|Default|GetSpawnedPickups) (Utilities|Array|LastIndex (Variables|Default|GetSpawnedPickups)))))

(fn GiveAdditionalAmmo
  (params (InventoryToGrantAmmo :type Fort%20Inventory%20Owner%20Interface%20Interface) (ItemNeedingAmmo :type Fort%20World%20Item%20Definition%20Object%20Reference) (Ammo%20Count :type Integer) (Ammo%20Override :type Boolean))
  (bind _ammoitemdefinition (LocalVariables|GetAmmoItemDefinition))
  (bind _output (Utilities|ResolveSoftReference _ammoitemdefinition))
  (bind _extraammocounts (Variables|Default|GetExtraAmmoCounts))
  (bind (_value _returnvalue) (Utilities|Map|Find _extraammocounts _output))
  (bind (_asfort_world_multi_item_definition _bsuccess) (Utilities|Casting|CastToFortWorldMultiItemDefinition ItemNeedingAmmo))
  (bind _inventorytograntammo (Variables|GetInventoryToGrantAmmo))
  (bind _output_get (LocalVariables|SetAmmoItemDefinition (select _bsuccess (Class|FortWorldMultiItemDefinition|GetDefaultAmmoWorldItemDefinition _asfort_world_multi_item_definition 0) (WeaponItemDefinition|GetAmmoWorldItemDefinition ItemNeedingAmmo))))
  (Utilities|IsValid _output_get
    (:Is%20Valid
      (if (and (Utilities|NotEqual(Object) (Variables|GetItemNeedingAmmo) (LocalVariables|GetAmmoItemDefinition)) (or (not (Variables|GetAmmo%20Override)) (> (Variables|GetAmmo%20Count) 0)))
        (if (Utilities|IsValidInterface (Variables|GetInventoryToGrantAmmo))
          (LocalVariables|SetCanBeDropped false)
          (bind _itemdata (Class|ItemDefinitionBase|GetItemData (LocalVariables|GetAmmoItemDefinition))
            (:Found
              (LocalVariables|SetCanBeDropped (.bCanBeDroppedFromInventory (Utilities|Struct|BreakFortnitePickupItemData _itemdata))))
            (:NotFound
              _))
          (if (and (== (Inventory|GetItemQuantityonInventoryOwner (Variables|GetInventoryToGrantAmmo) (LocalVariables|GetAmmoItemDefinition)) 0) (not (Variables|GetAmmo%20Override)))
            ;[1 1 1] Giving out ammo using Extra Ammo Counts map
            (if _returnvalue
              (Inventory|GiveItemtoInventoryOwner :InventoryOwner _inventorytograntammo :ItemDefinition (LocalVariables|GetAmmoItemDefinition) :NumberToGive _value :bNotifyPlayer true :ItemLevel -1 :PickupInstigatorHandle -1 :bUseItemPickupAnalyticEvent false :WeaponAmmoOverride -1)
              (elif (LocalVariables|GetCanBeDropped)
                (Inventory|GiveItemtoInventoryOwner :InventoryOwner _inventorytograntammo :ItemDefinition (LocalVariables|GetAmmoItemDefinition) :NumberToGive 5 :bNotifyPlayer true :ItemLevel -1 :PickupInstigatorHandle -1 :bUseItemPickupAnalyticEvent false :WeaponAmmoOverride -1)))
            (elif (Math|Boolean|ANDBoolean (Variables|GetAmmo%20Override) (> (Variables|GetAmmo%20Count) 0) (LocalVariables|GetCanBeDropped))
              ;[1 1 1] Giving out ammo using numerical value
              (Inventory|GiveItemtoInventoryOwner :InventoryOwner (Variables|GetInventoryToGrantAmmo) :ItemDefinition (LocalVariables|GetAmmoItemDefinition) :NumberToGive (Variables|GetAmmo%20Count) :bNotifyPlayer true :ItemLevel -1 :PickupInstigatorHandle -1 :bUseItemPickupAnalyticEvent false :WeaponAmmoOverride -1))))))))

(fn HasBegunSpawning
  (return (Variables|Default|GetbHasBegunSpawning)))

(fn HasReachedMaxSpawns
  (return (and (!= (Variables|Default|GetMaxSpawns) -1) (>= (Variables|Default|GetNumItemsSpawned) (Variables|Default|GetMaxSpawns)))))

(fn HasRoomToSpawn
  (return (> (Variables|Default|GetNumSimultaneous) (Utilities|Array|Length (Variables|Default|GetSpawnedPickups)))))

(fn IsNextSpawnIndexValid
  (return (Utilities|Array|IsValidIndex (Variables|Minigame|GetToSpawnList) (Variables|Default|GetNextSpawnIndex))))

(fn IsSpawning
  (return (and (Variables|Default|GetbHasBegunSpawning) (not (Variables|Default|GetbIsFinishedSpawning)))))

(fn MakeSpawnTransform
  (params (Actor :type Actor%20Object%20Reference) (SpawnLocation :type Vector))
  (return (Math|Transform|MakeTransform SpawnLocation (.rotation (Transformation|GetActorTransform Actor)) (.scale (Transformation|GetActorTransform Actor)))))

(fn OnPickup
  (params (PickupObject :type Fort%20Pickup%20Object%20Reference) (InteractingPawn :type Fort%20Pawn%20Object%20Reference))
  (bind _returnvalue (Pickup|GetPrimaryItemDefinition PickupObject))
  (bind _returnvalue_2 (Pawn|GetController InteractingPawn))
  (bind _asfort_inventory_owner_interface (Utilities|Casting|CastToFortInventoryOwnerInterface _returnvalue_2))
  (bind _asfort_world_item_definition (Utilities|Casting|CastToFortWorldItemDefinition _returnvalue))
  (Pickup|OnPickupRemovedfromPlay PickupObject)
  (CallFunction|GiveAdditionalAmmo _asfort_inventory_owner_interface _asfort_world_item_definition (Variables|Default|GetSpareWeaponAmmo) (Variables|Default|GetSpareWeaponAmmo_Override))
  (Utilities|FlowControl|Sequence
    (:then_0
      (bind _asfort_world_item_definition_2 (Utilities|Casting|CastToFortWorldItemDefinition (WeaponItemDefinition|GetLinkedItemDefinition _asfort_world_item_definition)))
      (CallFunction|GiveAdditionalAmmo _asfort_inventory_owner_interface _asfort_world_item_definition_2 (Variables|Default|GetSpareLinkedWeaponAmmo) (Variables|Default|GetSpareLinkedWeaponAmmo_Override)))
    (:then_1
      (Utilities|IsValid _returnvalue_2
        (:Is%20Valid
          (Default|CallOnPickedUpSpawnedItem _returnvalue_2))))))

(fn OnPickupCombined
  (params (SourcePickup :type Fort%20Pickup%20Object%20Reference) (TargetPickup :type Fort%20Pickup%20Object%20Reference))
  (Default|CallOnCombinedPickup)
  (Pickup|BindPickupDelegates TargetPickup))

(fn OnPickupCombinedOverflow
  (params (NewPickup :type Fort%20Pickup%20Object%20Reference) (ConsumedPickup :type Fort%20Pickup%20Object%20Reference))
  (Utilities|Array|RemoveItem (Variables|Default|GetSpawnedPickups) ConsumedPickup) ; Even though the consumed pickup will  get removed in the OnPickupDestroyed function, we remove it here to ensure we have space for the new pickup.
  (Pickup|SetupPickup NewPickup))

(fn OnPickupDespawned
  (params (Pickup :type Fort%20Pickup%20Object%20Reference))
  (Pickup|OnPickupRemovedfromPlay Pickup))

(fn OnPickupDestroyed
  (params (DestroyedActor :type Actor%20Object%20Reference))
  (bind _asfort_pickup (Utilities|Casting|CastToFortPickup DestroyedActor))
  (Pickup|OnPickupRemovedfromPlay _asfort_pickup))

(fn OnPickupRemovedFromPlay
  (params (Pickup :type Fort%20Pickup%20Object%20Reference))
  (bind _pickup (Variables|GetPickup))
  (Utilities|FlowControl|Sequence
    (:then_0
      (if (Math|Boolean|ANDBoolean (not (CallFunction|HasRoomtoSpawn)) (CallFunction|IsSpawning) (not (Variables|Default|GetbBlockSpawning)) (not (Variables|Default|GetbPauseSpawning)))
        (CallFunction|RestartTimer)))
    (:then_1
      (Utilities|Array|RemoveItem (Variables|Default|GetSpawnedPickups) _pickup)
      (Pickup|UnbindPickupDelegates _pickup)
      ;[1 1 1] Turn off VFX on spawner if it has spawned its last item
      (if (Variables|Default|GetbIsFinishedSpawning)
        (Default|CallOnLastItemPickedUp)))))

(fn PauseSpawner
  (if (not (Variables|Default|GetbPauseSpawning))
    (Variables|Default|SetbPauseSpawning true)
    (Utilities|Time|ClearandInvalidateTimerbyHandle (Variables|Default|GetSpawnTimer))
    (CallFunction|DestroyPickups)))

(fn RandomSpawnAndRemoveItem
  (params (ListToSpawn :type Array%20of%20Minigame%20Spawner%20Spawn%20Params%20Structures))
  (bind _output_get (LocalVariables|SetChosenIndex (Math|Random|RandomIntegerinRange 0 (Utilities|Array|LastIndex ListToSpawn)))) ; Ensures we remove the same item that was picked
  (if (Utilities|Array|IsValidIndex ListToSpawn _output_get)
    (CallFunction|Spawn (Utilities|Array|Get(aref) ListToSpawn _output_get))
    (Utilities|Array|RemoveIndex ListToSpawn (LocalVariables|GetChosenIndex))
    (else
      (Development|PrintString "WARNING:: Attempting to spawn an item when there are no items left to spawn"))))

(fn RandomSpawnEverythingOnce
  (params (HasListBeenCopied :type Boolean) (ListToSpawn :type Array%20of%20Minigame%20Spawner%20Spawn%20Params%20Structures))
  (bind _remainingspawnlist (Variables|Default|GetRemainingSpawnList))
  (if (not HasListBeenCopied)
    (bind _copiedspawnlist (CallFunction|CopySpawnListArray ListToSpawn))
    (Variables|Default|SetRemainingSpawnList _copiedspawnlist))
  (CallFunction|RandomSpawnandRemoveItem _remainingspawnlist)
  (return (or (<= (Utilities|Array|Length _remainingspawnlist) 0) (CallFunction|HasReachedMaxSpawns))))

(fn ResetSpawner
  (params (bDestroyExistingPickups :type Boolean :default true))
  (Variables|Default|SetNextSpawnIndex 0)
  (Variables|Default|SetNumItemsSpawned 0)
  (Utilities|Time|ClearandInvalidateTimerbyHandle (Variables|Default|GetSpawnTimer))
  (Variables|Default|SetbHasBegunSpawning false)
  (Variables|Default|SetbIsFinishedSpawning false)
  (Utilities|FlowControl|Sequence
    (:then_0
      (Utilities|Array|Clear (Variables|Default|GetRemainingSpawnList))
      (Variables|Default|SetHasToSpawnArrayBeenCopied false))
    (:then_1
      (if (Variables|GetbDestroyExistingPickups)
        (CallFunction|DestroyPickups)))
    (:then_2
      (Variables|Default|SetbBlockSpawning false))))

(fn RestartTimer
  (Utilities|Time|ClearandInvalidateTimerbyHandle (Variables|Default|GetSpawnTimer))
  ;[1 1 1] If the tick interval is too low, a looping timer can tick multiple times per frame ( which is unnecessary in this case ). So instead use a non-looping timer that restarts itself.
  (bind _returnvalue (Utilities|Time|SetTimerbyEvent (EventDispatchers|CreateEvent :!event Spawn%20Next%20and%20Restart%20Timer) (Variables|Default|GetDelayBetweenSpawns)))
  (Variables|Default|SetSpawnTimer _returnvalue))

(fn RestartTimerAndForceBeginSpawning
  ;[1 1 1] Force begin spawning to allow cases where spawner was NOT respawning, and finished spawning, but NOW we want to enable respawning later
  (Utilities|Time|ClearandInvalidateTimerbyHandle (Variables|Default|GetSpawnTimer))
  (Variables|Default|SetbHasBegunSpawning true)
  ;[1 1 1] If the tick interval is too low, a looping timer can tick multiple times per frame ( which is unnecessary in this case ). So instead use a non-looping timer that restarts itself.
  (bind _returnvalue (Utilities|Time|SetTimerbyEvent (EventDispatchers|CreateEvent :!event Spawn%20Next%20and%20Restart%20Timer) (Variables|Default|GetDelayBetweenSpawns)))
  (Variables|Default|SetSpawnTimer _returnvalue)
  (return))

(fn ResumeSpawner
  (if (and (Variables|Default|GetbPauseSpawning) (not (Variables|Default|GetbIsFinishedSpawning)))
    (Variables|Default|SetbPauseSpawning false)
    (CallFunction|ForceSpawnRemaining true)))

(fn SetPickupSettings
  (params (NewAllowedTeam :type Integer) (NewAllowedClass :type Integer) (NewAffectsTeam :type Integer) (NewAffectsClass :type Integer) (NewItemsVisibleToAllowedOnly :type Integer))
  (Variables|Default|SetAllowedTeam NewAllowedTeam)
  (Variables|Default|SetAllowedClass NewAllowedClass)
  (Variables|Default|SetAffectsTeam NewAffectsTeam)
  (Variables|Default|SetAffectsClass NewAffectsClass)
  (Variables|Default|SetItemsVisibleToAllowedOnly NewItemsVisibleToAllowedOnly)
  (for _array_element (Variables|Default|GetSpawnedPickups)
    (bind _asfort_pickup_creative (Utilities|Casting|CastToFort_Pickup_Creative _array_element))
    (Class|FortPickupCreative|SetPickupSettings _asfort_pickup_creative (Variables|Default|GetAllowedTeam) (Variables|Default|GetAllowedClass) (Variables|Default|GetAffectsTeam) (Variables|Default|GetAffectsClass) (Variables|Default|GetItemsVisibleToAllowedOnly))))

(fn SetSpawnerBlockedByPawn
  (params (bBlockedByPawn :type Boolean))
  (Variables|Default|SetbBlockedByPawn (and bBlockedByPawn (Variables|Default|GetIsPickupOnTouch))))

(fn SetupPickup
  (params (Pickup :type Fort%20Pickup%20Object%20Reference))
  (bind _pickup (Variables|GetPickup)
    (:then
      (Pickup|BindPickupDelegates _pickup)
      (bind _asfort_pickup_creative (Utilities|Casting|CastToFort_Pickup_Creative _pickup)
        (:then
          (Class|FortPickupCreative|SetPickupSettings _asfort_pickup_creative (Variables|Default|GetAllowedTeam) (Variables|Default|GetAllowedClass) (Variables|Default|GetAffectsTeam) (Variables|Default|GetAffectsClass) (Variables|Default|GetItemsVisibleToAllowedOnly)))
        (:CastFailed
          _))
      (Utilities|Array|Add (Variables|Default|GetSpawnedPickups) _pickup)
      (if (not (Variables|Default|GetbShouldApplyGravityToPickups))
        (Variables|/Script/Engine.ProjectileMovementComponent|SetProjectileGravityScale 0.0 (Variables|/Script/FortniteGame.FortPickup|GetMovementComponent _asfort_pickup_creative))))
    (:else)))

(fn SimpleRandomSpawn
  (params (List%20to%20Spawn :type Array%20of%20Minigame%20Spawner%20Spawn%20Params%20Structures))
  (if (> (Utilities|Array|Length List%20to%20Spawn) 0)
    (CallFunction|Spawn (Utilities|Array|Get(aref) List%20to%20Spawn (Math|Random|RandomIntegerinRange 0 (Utilities|Array|LastIndex List%20to%20Spawn))))))

(fn SpawnNextandRestartTimer
  (CallFunction|SpawnNext)
  (CallFunction|RestartTimer))

(fn SpawnManuallySelectedIndex
  (params (SelectedIndex :type Integer) (OptionalInventoryOwner :type Fort%20Inventory%20Owner%20Interface%20Interface))
  (bind _returnvalue (Components|GetOwner))
  (bind _output (Utilities|Array|Get(aref) (Variables|Minigame|GetToSpawnList) SelectedIndex))
  (bind _tospawnlist (Variables|Minigame|GetToSpawnList))
  (Utilities|IsValid _returnvalue
    (:Is%20Valid
      (if (Networking|HasAuthority _returnvalue)
        (if (CallFunction|CanSpawn)
          (Variables|Default|SetbBlockSpawning (Variables|Default|GetbUseManualSpawnControl))
          (if (Utilities|Array|IsValidIndex _tospawnlist SelectedIndex)
            (bind _succeeded (CallFunction|AttemptGrant (Variables|GetOptionalInventoryOwner) _output))
            (if _succeeded
              (return true)
              (else
                (CallFunction|Spawn _output)
                (return true)))
            (else
              _))
          (else
            _))
        (return false)
        (else
          _)))
    (:Is%20Not%20Valid
      _)))

(fn SpawnNext
  (bind _returnvalue (Components|GetOwner))
  (bind _randomspawnoption (Variables|Default|GetRandomSpawnOption))
  (Default|CallOnBeginSpawning)
  (Utilities|IsValid _returnvalue
    (:Is%20Valid
      (if (Networking|HasAuthority _returnvalue)
        (if (CallFunction|CanSpawn)
          (Variables|Default|SetbBlockSpawning (Variables|Default|GetbUseManualSpawnControl))
          (if (Variables|Default|GetbIsLooping)
            (switch int _randomspawnoption
              (:0
                (bind _hasfinishedspawning (CallFunction|SpawnSequentially (Variables|Minigame|GetToSpawnList)))
                (if _hasfinishedspawning
                  (Variables|Default|SetNextSpawnIndex 0)))
              (:1
                (CallFunction|SimpleRandomSpawn (Variables|Minigame|GetToSpawnList)))
              (:2
                (bind _hasfinishedspawning_2 (CallFunction|RandomSpawnEverythingOnce (Variables|Default|GetHasToSpawnArrayBeenCopied) (Variables|Minigame|GetToSpawnList)))
                (if _hasfinishedspawning_2
                  (Variables|Default|SetHasToSpawnArrayBeenCopied false))))
            (else
              (switch int _randomspawnoption
                (:0
                  (bind _hasfinishedspawning_3 (CallFunction|SpawnSequentially (Variables|Minigame|GetToSpawnList)))
                  (if _hasfinishedspawning_3
                    _))
                (:1
                  (bind _hasfinishedspawning_4 (CallFunction|RandomSpawnEverythingOnce (Variables|Default|GetHasToSpawnArrayBeenCopied) (Variables|Minigame|GetToSpawnList)))
                  (if _hasfinishedspawning_4
                    _))
                (:2
                  _))
              (CallFunction|FinishSpawning))))))))

(fn SpawnRemaining
  (params (bRestartTimer :type Boolean :default true))
  (if (not (CallFunction|IsSpawning))
    (CallFunction|ForceSpawnRemaining bRestartTimer)))

(fn SpawnSequentially
  (params (ListToSpawn :type Array%20of%20Minigame%20Spawner%20Spawn%20Params%20Structures))
  (if (Utilities|Array|IsValidIndex ListToSpawn (Variables|Default|GetNextSpawnIndex))
    (CallFunction|Spawn (Utilities|Array|Get(aref) (Variables|Minigame|GetToSpawnList) (Variables|Default|GetNextSpawnIndex)))
    (bind _result (Math|Integer|IncrementInt (Variables|Default|GetNextSpawnIndex))
      (:%20%20
        (return (or (>= _result (Utilities|Array|Length (Variables|Minigame|GetToSpawnList))) (CallFunction|HasReachedMaxSpawns)))))
    (else
      (return true))))

(fn TeamSettingToTeam
  (params (TeamIndex :type Integer))
  (switch int TeamIndex
    (:Default
      (return (Math|Byte|Byte+Byte (Utilities|Enum|LiteralenumEFortTeam "HumanPvP_Team1") (Math|Conversions|ToByte(Integer) (- (Variables|GetTeamIndex) 1)))))
    (:0
      (return (Utilities|Enum|LiteralenumEFortTeam "Spectator")))))

(fn TryOverrideSecondaryWeaponAmmoOnSpawnParams
  (params (Override :type Boolean) (InitialSecondaryWeaponAmmunition :type Integer))
  (bind _tospawnlist (Variables|Minigame|GetToSpawnList))
  (if Override
    (for (_ _array_index) _tospawnlist
      (Utilities|Struct|SetmembersinMinigameSpawnerSpawnParams (Utilities|Array|Get(aref) _tospawnlist _array_index) InitialSecondaryWeaponAmmunition))))

(fn TryOverrideWeaponAmmoOnSpawnParams
  (params (Override :type Boolean) (InitialWeaponAmmunition :type Integer))
  (bind _tospawnlist (Variables|Minigame|GetToSpawnList))
  (if Override
    (for (_ _array_index) _tospawnlist
      ; At this point we are expecting that the Item defination is already loaded
      (Utilities|IsValid (Utilities|ResolveSoftReference (Utilities|Struct|BreakMinigameSpawnerSpawnParams (Utilities|Array|Get(aref) _tospawnlist _array_index)))
        (:Is%20Valid
          (bind _asfort_weapon_item_definition (Utilities|Casting|CastToFortWeaponItemDefinition (Utilities|ResolveSoftReference (Utilities|Struct|BreakMinigameSpawnerSpawnParams (Utilities|Array|Get(aref) _tospawnlist _array_index))))
            (:then
              (if (GameplayTags|HasTag (Variables|/Script/FortniteGame.FortWeaponItemDefinition|GetAnalyticTags _asfort_weapon_item_definition) "(TagName=\"Item.Consumable\")")
                _
                (else
                  _)))
            (:CastFailed
              _))
          (Utilities|Struct|SetmembersinMinigameSpawnerSpawnParams (Utilities|Array|Get(aref) _tospawnlist _array_index) InitialWeaponAmmunition))
        (:Is%20Not%20Valid
          (Development|PrintString "The weapon item defination is not loaded"))))))

(fn UnbindPickupDelegates
  (params (Pickup :type Fort%20Pickup%20Object%20Reference))
  (bind _pickup (Variables|GetPickup)
    (:then
      (Interaction|UnbindEventfromOnPickup _pickup (EventDispatchers|CreateEvent :!event OnPickup))
      (Game|UnbindEventfromOnDestroyed _pickup (EventDispatchers|CreateEvent :!event OnPickupDestroyed))
      (Interaction|UnbindEventfromOnPickupCombinedOverflow _pickup (EventDispatchers|CreateEvent :!event OnPickupCombinedOverflow))
      (Interaction|UnbindEventfromOnPickupCombined _pickup (EventDispatchers|CreateEvent :!event OnPickupCombined)))
    (:else)))
