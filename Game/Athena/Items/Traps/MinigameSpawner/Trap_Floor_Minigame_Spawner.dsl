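; Creative item spawner floor trap: players drop pickups onto it to build a spawn list that it dispenses during minigames.
; The authority queues hit pickups, plays multicast FX, then moves them into the spawner component and destroys the originals.
; Minigame start, optionally delayed, and enable, disable, spawn and cycle events drive that component, with options copied into it.
; Clients show an item preview and pulsing emissive visuals reflecting spawn state.
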(event EventOnPlaced
  (Audio|PlaySoundatLocation "/Game/Sounds/Creative/Gadgets/Traps/Trap_ItemSpawner_Placed_Cue.Trap_ItemSpawner_Placed_Cue" (Transformation|GetActorLocation)))

(event EventBeginPlay
  (bind _minigame_spawner_component (Variables|Default|GetMinigame_Spawner_Component))
  (|CallParentFunction)
  (Networking|SwitchHasAuthority
    (:Authority
      ;[1 1 1] Binding dispatcher events from the Spawner Component
      (Default|BindEventtoAddedtoSpawnListDispatcher _minigame_spawner_component (delegate Custom|OnAddedToSpawnList))
      (Default|BindEventtoOnSpawnDispatcher _minigame_spawner_component (delegate Custom|OnSpawn))
      (Default|BindEventtoOnLastItemPickedUp _minigame_spawner_component (delegate Custom|OnLastItemPickedUp))
      (Default|BindEventtoOnPickedUpSpawnedItem _minigame_spawner_component (delegate Custom|Picked%20Up%20Item))
      (OptionData|BindEventtoOnAnyPropertyChanged (Variables|Default|GetToyOptionsComponent) (EventDispatchers|CreateEvent :!event ResaveActor))
      (OptionData|AssignOnPropertiesLoaded (Variables|Default|GetToyOptionsComponent) (EventDispatchers|CreateEvent :!event ResaveActor)))
    (:Remote
      (Initialization|InitializeComponent (Variables|Default|GetCreative_ItemPreview_Component) (Variables|Default|GetRotationSceneComponent) (Variables|Default|GetPreviewStaticMesh) (Variables|Default|GetPreviewSkeletalMesh) (Variables|Default|GetHolo%20Loop%20Cue) 0 0 :!class /Game/Creative/Devices/Common/Components/Creative_ItemPreview_Component.Creative_ItemPreview_Component_C))))

(event Custom|OnSpawn
  (params (Spawned%20Actor :type Actor%20Object%20Reference))
  ;[1 1 1] Binding dispatcher events from the Spawner Component
  (CallFunction|OnSpawnFX Spawned%20Actor))

(event Custom|Picked%20Up%20Item
  (params (Controller :type Controller%20Object%20Reference))
  ;[1 1 1] Binding dispatcher events from the Spawner Component
  (GameplayEvent|BroadcastEvent (Variables|UserOptions|Events|GetOn%20Item%20Picked%20Up) 0 (Utilities|Struct|MakeGameplayEventPickedUp Controller)))

(event Custom|OnAddedToSpawnList
  (params (Added%20Actor :type Actor%20Object%20Reference))
  ;[1 1 1] Binding dispatcher events from the Spawner Component
  (CallFunction|ResaveActor))

(event Custom|OnLastItemPickedUp
  ;[1 1 1] Binding dispatcher events from the Spawner Component
  (Variables|Default|SetNo%20Remaining%20Spawns true))

(event EventDestroyed
  (bind _holo_loop_cue (Variables|Default|GetHolo%20Loop%20Cue))
  (if (Audio|Components|Audio|IsPlaying _holo_loop_cue)
    (Audio|Components|Audio|FadeOut _holo_loop_cue 0.2 0.0 "Linear" :!class /Script/Engine.AudioComponent)))

(event MinigameStartedDelegate(FortMinigameProgress)
  (bind _initialspawndelay (Variables|Default|GetInitialSpawnDelay))
  ;[1 1 1] When minigame starts
  (Networking|SwitchHasAuthority
    (:Authority
      (Variables|Default|SetItemSpawnLocation (+ (Transformation|GetWorldLocation (Variables|Default|GetDropBox)) "0, 0,100.000000"))
      (CallFunction|InitializeDefaultSpawnList)
      (Variables|Default|SetNo%20Remaining%20Spawns (== (Utilities|Array|Length (Variables|/Script/FortniteGame.FortMinigameSpawnerComponent|GetToSpawnList (Variables|Default|GetMinigame_Spawner_Component))) 0))
      (if (Variables|Default|GetEnabledOnStart)
        (if (> _initialspawndelay 0.0)
          (bind _returnvalue (Utilities|Time|SetTimerbyEvent (delegate Custom|SetupMinigameStartSpawn) _initialspawndelay))
          (Variables|Default|SetMinigameStartSpawnTimer _returnvalue)
          (else
            (Utilities|Time|ClearandInvalidateTimerbyHandle (Variables|Default|GetMinigameStartSpawnTimer))
            (if (Utilities|Enum|Equal(Enum) (Variables|/Script/FortniteGame.FortMinigame|GetCurrentState (Variables|/Script/FortniteGame.FortMinigameLogicComponent|GetCurrentMinigame (Variables|Default|GetFortMinigameProgress))) "InProgress")
              (Class|MinigameSpawnerComponent|SpawnRemaining (Variables|Default|GetMinigame_Spawner_Component) true))))
        (else
          (Variables|Default|SetbDisabledSpawning true)
          (Class|MinigameSpawnerComponent|PauseSpawner (Variables|Default|GetMinigame_Spawner_Component)))))))

(event Custom|SetupMinigameStartSpawn
  (Utilities|Time|ClearandInvalidateTimerbyHandle (Variables|Default|GetMinigameStartSpawnTimer))
  (if (Utilities|Enum|Equal(Enum) (Variables|/Script/FortniteGame.FortMinigame|GetCurrentState (Variables|/Script/FortniteGame.FortMinigameLogicComponent|GetCurrentMinigame (Variables|Default|GetFortMinigameProgress))) "InProgress")
    (Class|MinigameSpawnerComponent|SpawnRemaining (Variables|Default|GetMinigame_Spawner_Component) true)))

(event MinigameEndedDelegate(FortMinigameProgress)
  ;[1 1 1] When minigame ends
  (Networking|SwitchHasAuthority
    (:Authority
      (Class|MinigameSpawnerComponent|ResetSpawner (Variables|Default|GetMinigame_Spawner_Component) true))))

(event OnComponentHit(DropBox)
  (params (HitComponent :type Primitive%20Component%20Object%20Reference) (OtherActor :type Actor%20Object%20Reference) (OtherComp :type Primitive%20Component%20Object%20Reference) (NormalImpulse :type Vector :default "0, 0, 0") (Hit :type Hit%20Result%20Structure%20%28by%20ref%29))
  (bind _itemstobeaddedqueue (Variables|Default|GetItemsToBeAddedQueue))
  (bind _itemstobeaddedqueue_2 (Variables|Default|GetItemsToBeAddedQueue))
  ;[1 1 1] Handles item being added from a hit event.  Hit captures all pick-ups
  (Networking|SwitchHasAuthority
    (:Authority
      (bind _output_get (Variables|Default|SetItemThatHit OtherActor))
      (bind _asfort_pickup_athena (Utilities|Casting|CastToFortPickupAthena _output_get))
      (Utilities|IsPickupValid _asfort_pickup_athena (Variables|Default|GetValidPickupGameplayTagContainer)
        (:IsValid
          (if (Interaction|BlueprintCanInteract (Pickup|GetPawnWhoDroppedPickup _asfort_pickup_athena))
            ;;[1 1 1] Every bounce of the pickup causes a hit trigger.  This ensures rest of script only fires once
            (if (not (Utilities|Equal(Object) (Variables|Default|GetItemThatHit) (Variables|Default|GetItemToBeAdded)))
              (bind _output_get_2 (Variables|Default|SetItemToBeAdded (Variables|Default|GetItemThatHit)))
              (bind _returnvalue (Utilities|Array|Add _itemstobeaddedqueue _output_get_2))
              (CallFunction|RPCMULTICASTAddItemFX (Utilities|Array|Get(acopy) _itemstobeaddedqueue _returnvalue))
              (Utilities|FlowControl|Delay 2.0) ; Ensure object still exists so FX can start before object is destroyed
              ;;[1 1 1] Loop through Queue array to be sure to add every item if items are added in rapid succession
              (if (> (Utilities|Array|Length _itemstobeaddedqueue_2) 0) :!label _label
                (CallFunction|AddtoSpawnList (Utilities|Array|Get(acopy) _itemstobeaddedqueue_2))
                (goto _label)))))))))

(event Custom|Enable%20Event
  (params (Instigator :type Controller%20Object%20Reference))
  (Networking|SwitchHasAuthority
    (:Authority
      (if (Utilities|Enum|Equal(Enum) (Variables|/Script/FortniteGame.FortMinigame|GetCurrentState (Variables|/Script/FortniteGame.FortMinigameLogicComponent|GetCurrentMinigame (Variables|Default|GetFortMinigameProgress))) "InProgress")
        (if (Variables|Default|GetbDisabledSpawning)
          (Variables|Default|SetbDisabledSpawning false)
          (Class|MinigameSpawnerComponent|ResumeSpawner (Variables|Default|GetMinigame_Spawner_Component))
          (Utilities|Time|ClearandInvalidateTimerbyHandle (Variables|Default|GetMinigameStartSpawnTimer)))))))

(event Custom|Disable%20Event
  (params (Instigator :type Controller%20Object%20Reference))
  (Networking|SwitchHasAuthority
    (:Authority
      (if (Utilities|Enum|Equal(Enum) (Variables|/Script/FortniteGame.FortMinigame|GetCurrentState (Variables|/Script/FortniteGame.FortMinigameLogicComponent|GetCurrentMinigame (Variables|Default|GetFortMinigameProgress))) "InProgress")
        (Class|MinigameSpawnerComponent|DestroyPickups (Variables|Default|GetMinigame_Spawner_Component))
        (if (not (Variables|Default|GetbDisabledSpawning))
          (Variables|Default|SetbDisabledSpawning true)
          (Class|MinigameSpawnerComponent|PauseSpawner (Variables|Default|GetMinigame_Spawner_Component))
          (Utilities|Time|ClearandInvalidateTimerbyHandle (Variables|Default|GetMinigameStartSpawnTimer)))))))

(event Custom|Spawn%20Item%20Event
  (params (Instigator :type Controller%20Object%20Reference))
  (Networking|SwitchHasAuthority
    (:Authority
      (if (Utilities|Enum|Equal(Enum) (Variables|/Script/FortniteGame.FortMinigame|GetCurrentState (Variables|/Script/FortniteGame.FortMinigameLogicComponent|GetCurrentMinigame (Variables|Default|GetFortMinigameProgress))) "InProgress")
        (if (not (Class|MinigameSpawnerComponent|HasBegunSpawning (Variables|Default|GetMinigame_Spawner_Component)))
          (Utilities|Time|ClearandInvalidateTimerbyHandle (Variables|Default|GetMinigameStartSpawnTimer))
          (Class|MinigameSpawnerComponent|SpawnRemaining (Variables|Default|GetMinigame_Spawner_Component) (not (Variables|Default|GetbDisabledSpawning)))
          (else
            (Class|MinigameSpawnerComponent|SpawnNext (Variables|Default|GetMinigame_Spawner_Component))))))))

(event Custom|Cycle%20to%20Next%20Item%20Event
  (params (Instigator :type Controller%20Object%20Reference))
  (bind _minigame_spawner_component (Variables|Default|GetMinigame_Spawner_Component))
  (Networking|SwitchHasAuthority
    (:Authority
      (if (Utilities|Enum|Equal(Enum) (Variables|/Script/FortniteGame.FortMinigame|GetCurrentState (Variables|/Script/FortniteGame.FortMinigameLogicComponent|GetCurrentMinigame (Variables|Default|GetFortMinigameProgress))) "InProgress")
        (if (not (Class|MinigameSpawnerComponent|HasBegunSpawning (Variables|Default|GetMinigame_Spawner_Component)))
          (Utilities|Time|ClearandInvalidateTimerbyHandle (Variables|Default|GetMinigameStartSpawnTimer))
          (Class|MinigameSpawnerComponent|SpawnRemaining (Variables|Default|GetMinigame_Spawner_Component) (not (Variables|Default|GetbDisabledSpawning)))
          (else
            (Class|MinigameSpawnerComponent|DestroyPickups _minigame_spawner_component)
            (Class|MinigameSpawnerComponent|SpawnNext _minigame_spawner_component)))))))

(event OnAnyPropertyChanged(ToyOptionsComponent)
  (Networking|CanExecuteCosmeticEvents
    (:True
      (Class|CreativeItemPreviewComponent|UpdateScale (Variables|Default|GetCreative_ItemPreview_Component) (Variables|Default|GetItemSpawnScale)))))

(event OnPropertiesLoaded(ToyOptionsComponent)
  (Networking|CanExecuteCosmeticEvents
    (:True
      (Class|CreativeItemPreviewComponent|UpdateScale (Variables|Default|GetCreative_ItemPreview_Component) (Variables|Default|GetItemSpawnScale)))))

(event Custom|PlaySetTrapActiveVisuals
  (params (Enable :type Boolean))
  ;[1 1 1] Turn emissive values on or off
  (if Enable
    (bind (_direction _trapvisuals) (Timeline|SetTrapActiveVisuals PlayFromStart)
      (:Update
        (Rendering|Material|SetVectorParameterValue (Variables|Default|GetTrapMID) "EmissiveMultiplier" (Math|Color|Lerp(LinearColor) :B (Variables|Default|GetActiveColor) :Alpha _trapvisuals) :!class /Script/Engine.MaterialInstanceDynamic)
        (Rendering|Material|SetScalarParameterValue (Variables|Default|GetRaysMID) "Fade" _trapvisuals :!class /Script/Engine.MaterialInstanceDynamic)))
    (else
      (Timeline|SetTrapActiveVisuals ReverseFromEnd))))

(event Custom|OnSpawnFX
  (params (Spawned%20Actor :type Actor%20Object%20Reference))
  ;[1 1 1] Fires when an item is spawned during the minigame.  Minigame must be active if you want to see/hear it occur.
  (Networking|CanExecuteCosmeticEvents
    (:True
      (Audio|PlaySoundatLocation (Variables|Default|GetItemSpawnSound) (Variables|Default|GetItemSpawnLocation)) ; ItemSpawnSound
      (Effects|SpawnEmitteratLocation :EmitterTemplate "/Game/Athena/Items/Traps/MinigameSpawner/P_ItemSpawner_CastItem.P_ItemSpawner_CastItem" :Location (Transformation|GetWorldLocation (Variables|Default|GetLightRays)) :Scale "1.000000,1.000000,1.000000" :bAutoDestroy true :PoolingMethod "AutoRelease" :bAutoActivateSystem true)
      ;[1 1 1] Pulse emissive when item is spit out
      (bind (_direction _pulse) (Timeline|PulseTrap PlayFromStart)
        (:Update
          (Rendering|Material|SetVectorParameterValue (Variables|Default|GetTrapMID) "EmissiveMultiplier" (Math|Color|Lerp(LinearColor) (Variables|Default|GetActiveColor) (Variables|Default|GetPulseColor) _pulse) :!class /Script/Engine.MaterialInstanceDynamic))))))

(event OnItemContainerUpdated(Minigame_Spawner_Component)
  ;[1 0.6535 0] Updating Preview
  (CallFunction|UpdateItemPreviewList))

(event Custom|RPC_MULTICAST%20Add%20Item%20FX
  (params (Item :type Actor%20Object%20Reference))
  ;[1 0.6535 0] Updating Preview
  (Networking|CanExecuteCosmeticEvents
    (:True
      (bind _asfort_pickup_athena (Utilities|Casting|CastToFortPickupAthena Item))
      (Class|CreativeItemPreviewComponent|ItemAddedToQueueFX (Variables|Default|GetCreative_ItemPreview_Component) _asfort_pickup_athena))))

(event OnAddedtoMinigame(FortMinigameProgress)
  (params (Minigame :type Fort%20Minigame%20Object%20Reference))
  ;[1 0.1082 0 0.25] Disable Collision on drop area during Edit mode, as overlaps are expensive.
  (Minigame|BindEventtoOnPlayModeChangedDelegate Minigame (delegate Custom|OnPlayModeChanged))
  (CallFunction|OnPlayModeChanged Minigame (Minigame|IsinPlayMode Minigame :!class /Script/FortniteGame.FortMinigame)))

(event Custom|OnPlayModeChanged
  (params (Minigame :type Fort%20Minigame%20Object%20Reference) (bIsInPlayMode :type Boolean))
  ;[1 0.1082 0 0.25] Disable Collision on drop area during Edit mode, as overlaps are expensive.
  (Collision|SetCollisionEnabled (Variables|Default|GetDropBox) (select bIsInPlayMode "NoCollision" "QueryOnly")))

(fn AddToSpawnList
  (params (InputPin :type Actor%20Object%20Reference))
  (Networking|SwitchHasAuthority
    (:Authority
      (Class|MinigameSpawnerComponent|AddtoSpawnList (Variables|Default|GetMinigame_Spawner_Component) InputPin (Variables|Default|GetItemSpawnLocation))
      (Actor|DestroyActor InputPin) ; Destroy the pick-up that is in the world.  Hidden by FX
      (Utilities|Array|RemoveIndex (Variables|Default|GetItemsToBeAddedQueue)))))

(fn GetVisualComponents
  (return (Utilities|Array|MakeArray (Variables|Debug|GetStaticMeshComponent) (Variables|Default|GetLightRays) (Variables|Default|GetPreviewStaticMesh) (Variables|Default|GetPreviewSkeletalMesh) (Variables|Default|GetHolo%20Loop%20Cue))))

(fn InitializeDefaultSpawnList
  ;[1 1 1] Handles setting up the spawn list to account for items that were put into the list from UnrealEd defaults
  (bind _result (Math|Integer|DecrementInt (Utilities|Array|Length (Variables|/Script/FortniteGame.FortMinigameSpawnerComponent|GetToSpawnList (Variables|Default|GetMinigame_Spawner_Component))))
    (:%20%20
      (for _index (range (+ _result 1))
        (Utilities|Struct|SetmembersinMinigameSpawnerSpawnParams (Utilities|Array|Get(aref) (Variables|/Script/FortniteGame.FortMinigameSpawnerComponent|GetToSpawnList (Variables|Default|GetMinigame_Spawner_Component)) _index) (Math|Transform|MakeTransform (Variables|Default|GetItemSpawnLocation)))))))

(fn IsVisibleDuringMinigame
  (return (Variables|Default|GetIsMeshVisibleDuringMinigame)))

(fn OnRep_bDisabledSpawning
  (CallFunction|UpdateSpawnerFX))

(fn OnRep_NoRemainingSpawns
  (CallFunction|UpdateSpawnerFX))

(fn ResaveActor
  (bind _minigame_spawner_component (Variables|Default|GetMinigame_Spawner_Component))
  (Variables|/Game/Athena/Items/Traps/MinigameSpawner/Minigame_Spawner_Component.Minigame_Spawner_Component_C|SetDelayBetweenSpawns (Variables|Default|GetTimeBetweenSpawnsSeconds) _minigame_spawner_component)
  (Variables|/Game/Athena/Items/Traps/MinigameSpawner/Minigame_Spawner_Component.Minigame_Spawner_Component_C|SetbIsLooping (Variables|Default|GetIsLooping) _minigame_spawner_component)
  (Variables|/Game/Athena/Items/Traps/MinigameSpawner/Minigame_Spawner_Component.Minigame_Spawner_Component_C|SetGrantExtraAmmo (Variables|Default|GetGrantExtraAmmoSetting) _minigame_spawner_component)
  (Variables|/Game/Athena/Items/Traps/MinigameSpawner/Minigame_Spawner_Component.Minigame_Spawner_Component_C|SetRandomSpawnOption (Variables|Default|GetRandomOption) _minigame_spawner_component)
  (Variables|/Game/Athena/Items/Traps/MinigameSpawner/Minigame_Spawner_Component.Minigame_Spawner_Component_C|SetIsPickupOnTouch (Variables|Default|GetIsPickupOnTouch) _minigame_spawner_component)
  (Variables|/Game/Athena/Items/Traps/MinigameSpawner/Minigame_Spawner_Component.Minigame_Spawner_Component_C|SetItemSpawnScale (Math|Vector|MakeVector (Variables|Default|GetItemSpawnScale) (Variables|Default|GetItemSpawnScale) (Variables|Default|GetItemSpawnScale)) _minigame_spawner_component)
  (CallFunction|Resave))

(fn UpdateItemPreviewList
  (Networking|CanExecuteCosmeticEvents
    (:True
      (Class|CreativeItemPreviewComponent|UpdateItemListFromSpawnParams (Variables|Default|GetCreative_ItemPreview_Component) (Variables|/Script/FortniteGame.FortMinigameSpawnerComponent|GetToSpawnList (Variables|Default|GetMinigame_Spawner_Component)))
      (CallFunction|UpdateSpawnerFX))))

(fn UpdateSpawnerFX
  (Variables|Default|SetTrapMID (Material|GetMID (Variables|Debug|GetStaticMeshComponent)))
  (Variables|Default|SetRaysMID (Material|GetMID (Variables|Default|GetLightRays)))
  (CallFunction|PlaySetTrapActiveVisuals (Math|Boolean|ANDBoolean (not (Variables|Default|GetbDisabledSpawning)) (Math|Conversions|ToBoolean(Integer) (Utilities|Array|Length (Variables|/Script/FortniteGame.FortMinigameSpawnerComponent|GetToSpawnList (Variables|Default|GetMinigame_Spawner_Component)))) (not (Variables|Default|GetNo%20Remaining%20Spawns)))))

(fn ConstructionScript
  (Networking|CanExecuteCosmeticEvents))
